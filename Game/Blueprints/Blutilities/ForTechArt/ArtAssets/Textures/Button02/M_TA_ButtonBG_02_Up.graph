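# Unlit UI material that draws the background of a tech-art button, named for its Up state.
# A packed mask texture's R, G and B channels drive chained lerps between four color layers: edge, interior, ring edge and ring interior.
# Each layer derives from the shared BaseColor_DoNotSetOnInstances parameter through its own multiplier and desaturation controls.
# Both ring layers add hue shift, and the interior darkens along a vertical UV gradient.
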Material /Game/Blueprints/Blutilities/ForTechArt/ArtAssets/Textures/Button02/M_TA_ButtonBG_02_Up
graph {
 properties {materialDomain=MD_UI, pixelDepthOffsetMode=PDOM_Legacy, shadingModel=MSM_Unlit, usedShadingModels=Unlit}
 nodes [
  n1 = NamedRerouteUsage()
  n2 = NamedRerouteUsage()
  n3 = ComponentMask(None=NamedRerouteUsage()) {r=true}
  n4 = NamedRerouteUsage()
  LinearInterpolate(
   A=LinearInterpolate(A=LinearInterpolate(A=@n1, B=@n2, Alpha=@n3), B=NamedRerouteUsage(), Alpha=ComponentMask(None=@n4) {g=true})
   B=NamedRerouteUsage()
   Alpha=ComponentMask(None=NamedRerouteUsage()) {b=true}
  )
  NamedRerouteDeclaration(None=TextureSample() {samplerType=SAMPLERTYPE_LinearColor, texture=&/T_TA_ButtonBG_02_M}.RGBA) {name=mask, nodeColor=<rgba 0.0312 1 0 1>}
  n5 = VectorParameter() {group=ZDoNotSet, parameterName=BaseColor_DoNotSetOnInstances, defaultValue=<rgba 0.1541 0.2535 0.8333 1>}
  n6 = ScalarParameter() {defaultValue=0.5, group=BaseEdge, parameterName=BaseEdge_Mult, sortPriority=20}
  NamedRerouteDeclaration(None=Desaturation(None=Multiply(A=@n5.RGB, B=@n6) {constB=0.5}, Fraction=ScalarParameter() {group=BaseEdge, parameterName=BaseEdge_Desat, sortPriority=21})) {name=edge, nodeColor=<rgba 0.2422 1 0 1>}
  n7 = VectorParameter() {group=ZDoNotSet, parameterName=BaseColor_DoNotSetOnInstances, defaultValue=<rgba 0.1541 0.2535 0.8333 1>}
  n8 = MaterialFunctionCall(Hue Shift Percentage=ScalarParameter() {group=RingEdge, parameterName=RingEdge_HueShift, sortPriority=41}, Texture=@n7.RGB) {materialFunction=&HueShift}
  n9 = ScalarParameter() {defaultValue=0.8, group=RingEdge, parameterName=RingEdge_Mult, sortPriority=40}
  NamedRerouteDeclaration(None=Desaturation(None=Multiply(A=@n8.Result, B=@n9) {constB=0.5}, Fraction=ScalarParameter() {group=RingEdge, parameterName=RingEdge_Desat, sortPriority=42})) {name=ringEdge, nodeColor=<rgba 0.7812 1 0 1>}
  n10 = VectorParameter() {group=ZDoNotSet, parameterName=BaseColor_DoNotSetOnInstances, defaultValue=<rgba 0.1541 0.2535 0.8333 1>}
  n11 = MaterialFunctionCall(Hue Shift Percentage=ScalarParameter() {group=RingInterior, parameterName=RingInterior_HueShift, sortPriority=31}, Texture=@n10.RGB) {materialFunction=&HueShift}
  n12 = ScalarParameter() {defaultValue=0.5, group=RingInterior, parameterName=RingInterior_Mult, sortPriority=30}
  NamedRerouteDeclaration(None=Desaturation(None=Multiply(A=@n11.Result, B=@n12) {constB=0.5}, Fraction=ScalarParameter() {group=RingInterior, parameterName=RingInterior_Desat})) {name=ringInterior, nodeColor=<rgba 0 0.8125 1 1>}
  n13 = VectorParameter() {group=ZDoNotSet, parameterName=BaseColor_DoNotSetOnInstances, defaultValue=<rgba 0.1541 0.2535 0.8333 1>}
  n14 = TextureCoordinate()
  n15 = Multiply(A=@n13.RGB, B=LinearInterpolate(Alpha=ComponentMask(None=@n14) {g=true}) {constA=1, constB=0.75})
  n16 = ScalarParameter() {defaultValue=1, group=BaseInterior, parameterName=BaseInterior_Mult, sortPriority=10}
  NamedRerouteDeclaration(None=Desaturation(None=Multiply(A=@n15, B=@n16), Fraction=ScalarParameter() {group=BaseInterior, parameterName=BaseInterior_Desat, sortPriority=11})) {name=interior, nodeColor=<rgba 0 1 0.9297 1>}
 ]
}
